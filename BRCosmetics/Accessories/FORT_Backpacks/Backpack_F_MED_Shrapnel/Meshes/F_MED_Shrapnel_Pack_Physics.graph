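# Physics asset for the Shrapnel backpack cosmetic mesh, defining its collision and joint setup for ragdoll or secondary physics motion.
# It holds a box body on the handle1 bone and a small capsule body on the pack bone.
# A single constraint joins handle1 to pack, with narrow, equal limits on both swing axes and twist, so relative movement stays slight.
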PhysicsAsset /BRCosmetics/Accessories/FORT_Backpacks/Backpack_F_MED_Shrapnel/Meshes/F_MED_Shrapnel_Pack_Physics
bodies {
 handle1 [{shapeType=Box, extentX=7.5, extentY=1.5, extentZ=4, center=<0 0 1.0615>}]
 pack [{shapeType=Capsule, length=1, radius=1, center=<0.6887 -5.4852 0.2629>, rotation=<rot 0 79.8844 -81.7129>}]
}
constraints [{bone1Name=handle1, bone2Name=pack, swing1LimitDegrees=15, swing2LimitDegrees=15, twistLimitDegrees=15}]
